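; IO Slider elevator that rises from the ground when a player approaches and releases AI or a chest.
; The server rolls ScalableFloat chances against Living World Manager budget, then enables patrol point providers or blocks spawning on a timer.
; A delayed fallback opens doors for partial spawns or substitutes a chest.
; Timelines drive lift and doors, clients add Niagara effects, and replicated ElevatorSpawned snaps unmoved clients to the raised position.
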(event EventBeginPlay
  ;[1 1 1] Hide Debug Components
  (Rendering|SetVisibility (Variables|Default|GetDebugElevator) false false)
  ;[1 1 1] Client Persistence
  (Utilities|FlowControl|Sequence
    (:then_0
      (CallFunction|SetOriginalAssetLocations))
    (:then_1
      (Networking|SwitchHasAuthority
        (:Authority
          ;[1 1 1] If Elevator Should not be Enabled, Destroy me
          (Class|FortAsyncActionCurrentPlaylistReady|PlaylistDataReadyAsync
            (:OnReady
              (if (!= (ScalableFloat|GetValueatLevel (Variables|Default|GetHD_ElevatorEnabled)) 1.0)
                (Actor|DestroyActor self)))))
        (:Remote
          (if (Variables|LocalLogic|GetElevatorSpawned)
            (CallFunction|TeleportElevatoratEnd)))))))

(event Custom|DisableSpawning
  ;[1 1 1] If Not Spawn, Set Timer to Re-Enable Spawning
  (Variables|LocalLogic|SetSpawnType 4)
  (Variables|LocalLogic|SetSpawningDisabled true)
  (Utilities|Time|SetTimerbyEvent (delegate Custom|UnblockSpawning) (ScalableFloat|GetValueatLevel (Variables|LocalLogic|GetHAND_RespawnDelay)))
  (CallFunction|FireSpawnAnalyticEvent))

(event Collision|EventActorBeginOverlap
  (params (OtherActor :type Actor%20Object%20Reference)))

(event Custom|UnblockSpawning
  ;[1 1 1] If Not Spawn, Set Timer to Re-Enable Spawning
  (Variables|LocalLogic|SetSpawningDisabled false))

(event OnComponentBeginOverlap(ProximityTrigger)
  (params (OverlappedComponent :type Primitive%20Component%20Object%20Reference) (OtherActor :type Actor%20Object%20Reference) (OtherComp :type Primitive%20Component%20Object%20Reference) (OtherBodyIndex :type Integer) (bFromSweep :type Boolean) (SweepResult :type Hit%20Result%20Structure%20%28by%20ref%29))
  (bind _hand_chancemspawn (Variables|LocalLogic|GetHAND_ChanceMSpawn))
  (bind _returnvalue (ScalableFloat|GetValueatLevel _hand_chancemspawn))
  (bind _returnvalue_2 (Math|Random|RandomFloatinRange 0.0 1.0))
  (bind _returnvalue_3 (<= _returnvalue_2 _returnvalue))
  ;[1 1 1] When Player Pawn is Near
  (Networking|SwitchHasAuthority
    (:Authority
      (bind _asfort_player_pawn (Utilities|Casting|CastToFortPlayerPawn OtherActor))
      (if (Pawn|IsBotControlled _asfort_player_pawn)
        _
        (elif (Variables|LocalLogic|GetSpawningDisabled)
          _
          (elif (Variables|Debug|GetOverrideChanceToSpawn)
            _
            (elif (<= (Math|Random|RandomFloatinRange 0.0 1.0) (ScalableFloat|GetValueatLevel (Variables|LocalLogic|GetHAND_ChanceToNotSpawn)))
              ;[1 1 1] If Not Spawn, Set Timer to Re-Enable Spawning
              (Variables|LocalLogic|SetSpawnType 5)
              (Variables|LocalLogic|SetSpawningDisabled true)
              (Utilities|Time|SetTimerbyEvent (delegate Custom|UnblockSpawning) (ScalableFloat|GetValueatLevel (Variables|LocalLogic|GetHAND_RespawnDelay)))
              (CallFunction|FireSpawnAnalyticEvent)))))
      ;[1 1 1] Logic to see what to Spawn
      (Variables|Default|SetLWM (Actor|GetComponentbyClass (Game|GetGameState) "/Script/LagerRuntime.FortAthenaLivingWorldManager"))
      (bind _budget (CallFunction|GetBudgetbyEvent (Variables|Default|GetHD_GruntTableCode)))
      (if (> _budget 1)
        (if (and (>= _budget 2) (<= (Math|Random|RandomFloatinRange 0.0 1.0) (ScalableFloat|GetValueatLevel (Variables|LocalLogic|GetHAND_ChanceLSpawn))))
          (Variables|LocalLogic|SetSpawnType 2)
          (elif _returnvalue_3
            (Variables|LocalLogic|SetSpawnType 1 :!label _label)
            (else
              (Variables|LocalLogic|SetSpawnType 0))))
        (CallFunction|StartRumbling)
        (elif _returnvalue_3
          _
          (else
            (CallFunction|DisableSpawning))))
      (goto _label))))

(event Custom|StartRumbling
  (bind _hand_anticipationdelaymin (Variables|LocalLogic|GetHAND_AnticipationDelayMin))
  (bind _returnvalue (ScalableFloat|GetValueatLevel _hand_anticipationdelaymin))
  (bind _hand_anticipationdelaymax (Variables|LocalLogic|GetHAND_AnticipationDelayMax))
  (bind _returnvalue_2 (ScalableFloat|GetValueatLevel _hand_anticipationdelaymax))
  (bind _returnvalue_3 (Math|Random|RandomFloatinRange _returnvalue _returnvalue_2))
  ;[1 1 1] Start Pre-Spawn Phase
  (Components|DestroyComponent (Variables|Default|GetProximityTrigger)) ; Remove Component, should only be used once, and can save perfs for Overlaps.
  (Networking|SwitchHasAuthority
    (:Authority
      (Utilities|FlowControl|Delay _returnvalue_3)
      (CallFunction|RaiseElevator))
    (:Remote
      (Niagara|SpawnSystematLocation (Variables|Default|GetNS_ElevatorExit_01_FirstImpulse) (Variables|/Script/Engine.SceneComponent|GetRelativeLocation (Variables|Default|GetDefaultSceneRoot)))
      (Variables|Default|SetRumbleTime _returnvalue_3)
      (CallFunction|PlayRumbleAudio))))

(event Custom|Play%20Rumble%20Audio
  ;[1 1 1] Update Audio
  (Audio|Components|Audio|Play (Variables|Default|GetRumble%20Audio) 0.0)
  (bind _rumble_audio (Variables|Default|GetRumble%20Audio)
    (:then
      (if (Audio|Components|Audio|IsPlaying _rumble_audio)
        (Audio|Components|Audio|AdjustVolume _rumble_audio (Variables|Default|GetRumbleTime) 2.0 "Linear" :!class /Script/Engine.AudioComponent)))
    (:else)))

(event Custom|Stop%20Audio
  ;[1 1 1] Update Audio
  (bind _rumble_audio (Variables|Default|GetRumble%20Audio)
    (:then
      (if (Audio|Components|Audio|IsPlaying _rumble_audio)
        (Audio|Components|Audio|FadeOut _rumble_audio 0.2 0.0 "Linear" :!class /Script/Engine.AudioComponent)))
    (:else)))

(event Custom|RaiseElevator
  (bind _baseoriginalposition (Variables|LocalLogic|GetBaseOriginalPosition))
  (bind _x (.x _baseoriginalposition))
  (bind _y (.y _baseoriginalposition))
  (bind _z (.z _baseoriginalposition))
  (bind _baseoriginalposition_2 (Variables|LocalLogic|GetBaseOriginalPosition))
  (bind _x_2 (.x _baseoriginalposition_2))
  (bind _y_2 (.y _baseoriginalposition_2))
  (bind _z_2 (.z _baseoriginalposition_2))
  (bind _baseoffset (Variables|LocalLogic|GetBaseOffset))
  (bind _returnvalue (+ _z _baseoffset))
  (bind _elevatorfxoffset (Variables|Default|GetElevatorFXOffset))
  (bind _defaultsceneroot (Variables|Default|GetDefaultSceneRoot))
  (bind _relativelocation (Variables|/Script/Engine.SceneComponent|GetRelativeLocation _defaultsceneroot))
  ;[1 1 1] Raise the Elevator from the Ground
  (Ability|SendGameplayEventtoActor self "(TagName=\"Event.IO.Elevator.KnockBack\")")
  (Utilities|FlowControl|Delay 0.2)
  (CallFunction|StopAudio)
  (Audio|PlaySoundatLocation (Variables|Sound|GetRaiseSound) _relativelocation)
  (Networking|SwitchHasAuthority
    (:Authority
      _)
    (:Remote
      (Niagara|SpawnSystematLocation (Variables|Default|GetNS_ElevatorExit_02_BreakThroughGround) _relativelocation)
      (bind _returnvalue_2 (Niagara|SpawnSystematLocation (Variables|Default|GetNS_ElevatorExit_03_RaiseUp) (Math|Vector|MakeVector (.x (Variables|LocalLogic|GetBaseOriginalPosition)) (.y (Variables|LocalLogic|GetBaseOriginalPosition)) (+ (.z (Variables|LocalLogic|GetBaseOriginalPosition)) (Variables|Default|GetFX_Elevator_Z_Offset)))))
      (Variables|Default|SetElevatorFXOffset _returnvalue_2)))
  (bind (_direction _elevatorposition) (Timeline|MoveUpElevator Play)
    (:Update
      (Utilities|IsValid _elevatorfxoffset
        (:Is%20Valid
          (Transformation|SetRelativeLocation _elevatorfxoffset (Math|Vector|MakeVector _x_2 _y_2 (+ (+ _elevatorposition _z_2) (Variables|Default|GetFX_Elevator_Z_Offset))) false false))
        (:Is%20Not%20Valid
          _))
      (Transformation|SetWorldLocation (Variables|Default|GetBase) (Math|Vector|MakeVector _x_2 _y_2 (+ _elevatorposition _z_2)) false false)
      (Networking|SwitchHasAuthority
        (:Authority
          (Utilities|FlowControl|DoOnce false
            (:Completed
              (Utilities|FlowControl|Delay 0.1)
              (Variables|LocalLogic|SetElevatorSpawned true)
              (Variables|Default|SetNewBaseLocation (Math|Vector|MakeVector _x _y _returnvalue))
              (Variables|Default|SetNewDoorLocation (Math|Vector|MakeVector _x _y (- _returnvalue (Variables|LocalLogic|GetDoorOffset)))))))))
    (:Finished
      (CallFunction|SpawnContents))))

(event Custom|SpawnContents
  (bind _patrolproviders (Variables|Patrol|GetPatrolProviders))
  (bind _patrolprovider_ls (Variables|Patrol|GetPatrolProvider_LS))
  (bind _patrolprovider_lg (Variables|Patrol|GetPatrolProvider_LG))
  ;[1 1 1] Spawn the AI and track when they are spawned.
  (Networking|SwitchHasAuthority
    (:Authority
      (switch int (Variables|LocalLogic|GetSpawnType)
        (:0
          (CallFunction|CalculateSpawnAmount)
          (Utilities|Array|Shuffle (Variables|Patrol|GetPatrolProviders))
          (for _index (range (Variables|LocalLogic|GetAmountToSpawn))
            (if (< _index (Utilities|Array|Length (Variables|Patrol|GetPatrolProviders)))
              (Utilities|IsValid (Utilities|Array|Get(acopy) _patrolproviders _index)
                (:Is%20Valid
                  (Class|FortAthenaLivingWorldPointProviderInterface|EnablePointProvider_DEPRECATED (Utilities|Array|Get(acopy) _patrolproviders _index))
                  (EventDispatchers|BindEventtoOnActorSpawnedResult (Utilities|Array|Get(acopy) _patrolproviders _index) (delegate Custom|LWMSpawnedActor) :!class /Script/LagerRuntime.FortAthenaPatrolPathPointProvider)))))
          ;[1 1 1] Force Spawn LWM
          (Class|FortAthenaLivingWorldManager|RequestEventGeneration (Variables|Default|GetLWM))
          (CallFunction|ValidateFallback)
          (CallFunction|PlayIOGuardAnnoucnementAudio))
        (:1
          (CallFunction|SpawnChest))
        (:2
          ;[1 1 1] Spawn L Guys
          (Utilities|IsValid _patrolprovider_ls
            (:Is%20Valid
              (Utilities|IsValid _patrolprovider_lg
                (:Is%20Valid
                  (Class|FortAthenaLivingWorldPointProviderInterface|EnablePointProvider_DEPRECATED _patrolprovider_ls)
                  (Class|FortAthenaLivingWorldPointProviderInterface|EnablePointProvider_DEPRECATED _patrolprovider_lg)
                  (EventDispatchers|BindEventtoOnActorSpawnedResult _patrolprovider_ls (delegate Custom|LWMSpawnedL) :!class /Script/LagerRuntime.FortAthenaPatrolPathPointProvider)
                  (EventDispatchers|BindEventtoOnActorSpawnedResult _patrolprovider_lg (delegate Custom|LWMSpawnedL) :!class /Script/LagerRuntime.FortAthenaPatrolPathPointProvider)
                  (Class|FortAthenaLivingWorldManager|RequestEventGeneration (Variables|Default|GetLWM))
                  (CallFunction|ValidateFallback)
                  (CallFunction|PlayIOGuardAnnoucnementAudio))
                (:Is%20Not%20Valid
                  _)))
            (:Is%20Not%20Valid
              _))
          (CallFunction|SpawnChest))))))

(event Custom|LWMSpawnedActor
  (params (SpawnedActor :type Actor%20Object%20Reference) (bResult :type Boolean))
  (bind _amountspawned (Variables|LocalLogic|GetAmountSpawned))
  ;[1 1 1] Spawn the AI and track when they are spawned.
  (if bResult
    (Math|Integer|IncrementInt _amountspawned
      (:%20%20
        (if (== (Variables|LocalLogic|GetAmountToSpawn) _amountspawned)
          (Variables|LocalLogic|SetAllSpawned true)
          (CallFunction|OpenDoors)
          (CallFunction|FireSpawnAnalyticEvent))))))

(event Custom|Play%20IO%20Guard%20Annoucnement%20Audio
  (Audio|PlaySoundatLocation "/Lager/Sounds/SFX/Elevator/IOSlider_Announcements_Cue.IOSlider_Announcements_Cue" (Transformation|GetActorLocation)))

(event Custom|ValidateFallback
  ;[1 1 1] Fallback Logic
  (Utilities|FlowControl|Delay (ScalableFloat|GetValueatLevel (Variables|LocalLogic|GetHAND_FallbackDelay)))
  (if (not (Variables|LocalLogic|GetAllSpawned))
    (CallFunction|DisablePointProviders)
    (if (> (Variables|LocalLogic|GetAmountSpawned) 0)
      (CallFunction|OpenDoors)
      (else
        (CallFunction|SpawnChest)))))

(event Custom|LWMSpawnedL
  (params (SpawnedActor :type Actor%20Object%20Reference) (bResult :type Boolean))
  ;[1 1 1] Spawn L Guys
  (if bResult
    (bind _result (Math|Integer|IncrementInt (Variables|LocalLogic|GetAmountSpawned))
      (:%20%20
        (if (== _result 2)
          (Variables|LocalLogic|SetAllSpawned true)
          (CallFunction|OpenDoors)
          (CallFunction|FireSpawnAnalyticEvent))))))

(event Custom|SpawnChest
  (bind _patrolprovider_m (Variables|Patrol|GetPatrolProvider_M))
  (bind _patrolprovider_m_2 (Variables|Patrol|GetPatrolProvider_M))
  ;[1 1 1] Spawn Chest / Museum
  (Utilities|IsValid (Variables|Default|GetLWM)
    (:Is%20Valid
      (Utilities|IsValid _patrolprovider_m
        (:Is%20Valid
          (Class|FortAthenaLivingWorldPointProviderInterface|EnablePointProvider_DEPRECATED _patrolprovider_m)
          (EventDispatchers|BindEventtoOnActorSpawnedResult _patrolprovider_m (delegate Custom|SpawnedMActor) :!class /Script/LagerRuntime.FortAthenaPatrolPathPointProvider)
          (Class|FortAthenaLivingWorldManager|RequestEventGeneration (Variables|Default|GetLWM))
          (Utilities|FlowControl|Delay (ScalableFloat|GetValueatLevel (Variables|LocalLogic|GetHAND_FallbackDelay)))
          (Utilities|FlowControl|Gate false
            (:Exit
              _)))
        (:Is%20Not%20Valid
          _)))
    (:Is%20Not%20Valid
      _))
  (Utilities|FlowControl|DoOnce false
    (:Completed
      (Game|SpawnActorfromClass "/Game/Building/ActorBlueprints/Containers/Tiered_Chest_Athena.Tiered_Chest_Athena_C" (Math|Transform|MakeTransform (Math|Vector|MakeVector (.x (.location (Transformation|GetActorTransform self))) (.y (.location (Transformation|GetActorTransform self))) (+ (.z (.location (Transformation|GetActorTransform self))) 12.0)) (.rotation (Transformation|GetActorTransform self)) (.scale (Transformation|GetActorTransform self))) "AlwaysSpawn" "MultiplyWithRoot" 0 0 0 0)
      (Variables|LocalLogic|SetSpawnType 3)
      (EventDispatchers|UnbindEventfromOnActorSpawnedResult _patrolprovider_m_2 (delegate Custom|SpawnedMActor) :!class /Script/LagerRuntime.FortAthenaPatrolPathPointProvider)
      (Class|FortAthenaPatrolPathPointProvider|DisablePointProvider _patrolprovider_m_2)
      (CallFunction|OpenDoors)
      (CallFunction|FireSpawnAnalyticEvent))))

(event Custom|SpawnedMActor
  (params (SpawnedActor :type Actor%20Object%20Reference) (bResult :type Boolean))
  (bind _patrolprovider_m (Variables|Patrol|GetPatrolProvider_M))
  ;[1 1 1] Spawn Chest / Museum
  (if bResult
    (Utilities|FlowControl|Sequence
      (:then_0
        (Utilities|FlowControl|Gate false
          (:Exit
            (Utilities|FlowControl|DoOnce false :!label _label
              (:Completed
                (Game|SpawnActorfromClass "/Game/Building/ActorBlueprints/Containers/Tiered_Chest_Athena.Tiered_Chest_Athena_C" (Math|Transform|MakeTransform (Math|Vector|MakeVector (.x (.location (Transformation|GetActorTransform self))) (.y (.location (Transformation|GetActorTransform self))) (+ (.z (.location (Transformation|GetActorTransform self))) 12.0)) (.rotation (Transformation|GetActorTransform self)) (.scale (Transformation|GetActorTransform self))) "AlwaysSpawn" "MultiplyWithRoot" 0 0 0 0)
                (Variables|LocalLogic|SetSpawnType 3)
                (EventDispatchers|UnbindEventfromOnActorSpawnedResult _patrolprovider_m (delegate Custom|SpawnedMActor) :!class /Script/LagerRuntime.FortAthenaPatrolPathPointProvider)
                (Class|FortAthenaPatrolPathPointProvider|DisablePointProvider _patrolprovider_m))))))
      (:then_1
        (Variables|LocalLogic|SetSpawnType 1)))
    (CallFunction|OpenDoors)
    (CallFunction|FireSpawnAnalyticEvent))
  (goto _label))

(event Custom|OpenDoors
  (bind _doorsoriginalposition (Variables|LocalLogic|GetDoorsOriginalPosition))
  (bind _x (.x _doorsoriginalposition))
  (bind _y (.y _doorsoriginalposition))
  (bind _z (.z _doorsoriginalposition))
  ;[1 1 1] Open the Doors to the Elevator When Spawning is Complete
  (Utilities|FlowControl|Delay 0.3)
  (Variables|LocalLogic|SetBarOriginalPositionFront (Variables|/Script/Engine.SceneComponent|GetRelativeLocation (Variables|Default|GetSM_Slider_Bar_Front)))
  (Variables|LocalLogic|SetBarOriginalPositionBack (Variables|/Script/Engine.SceneComponent|GetRelativeLocation (Variables|Default|GetSM_Slider_Bar_Rear)))
  (Variables|LocalLogic|SetBarOriginalPositionRight (Variables|/Script/Engine.SceneComponent|GetRelativeLocation (Variables|Default|GetSM_Slider_Bar_Right)))
  (Variables|LocalLogic|SetBarOriginalPositionLeft (Variables|/Script/Engine.SceneComponent|GetRelativeLocation (Variables|Default|GetSM_Slider_Bar_Left)))
  (Audio|PlaySoundatLocation (Variables|Sound|GetDoorOpenSound) (+ (Transformation|GetActorLocation) "0, 0,-500.000000"))
  (bind (_direction _doormidposition) (Timeline|RetractMiddleDoors Play)
    (:Update
      (Transformation|SetRelativeLocation (Variables|Default|GetSM_Slider_Bar_Front) (Math|Vector|MakeVector (.x (Variables|LocalLogic|GetBarOriginalPositionFront)) (- (.y (Variables|LocalLogic|GetBarOriginalPositionFront)) _doormidposition) (.z (Variables|LocalLogic|GetBarOriginalPositionFront))) false false)
      (Transformation|SetRelativeLocation (Variables|Default|GetSM_Slider_Bar_Rear) (Math|Vector|MakeVector (.x (Variables|LocalLogic|GetBarOriginalPositionBack)) (+ (.y (Variables|LocalLogic|GetBarOriginalPositionBack)) _doormidposition) (.z (Variables|LocalLogic|GetBarOriginalPositionBack))) false false)
      (Transformation|SetRelativeLocation (Variables|Default|GetSM_Slider_Bar_Right) (Math|Vector|MakeVector (+ (.x (Variables|LocalLogic|GetBarOriginalPositionRight)) _doormidposition) (.y (Variables|LocalLogic|GetBarOriginalPositionRight)) (.z (Variables|LocalLogic|GetBarOriginalPositionRight))) false false)
      (Transformation|SetRelativeLocation (Variables|Default|GetSM_Slider_Bar_Left) (Math|Vector|MakeVector (- (.x (Variables|LocalLogic|GetBarOriginalPositionLeft)) _doormidposition) (.y (Variables|LocalLogic|GetBarOriginalPositionLeft)) (.z (Variables|LocalLogic|GetBarOriginalPositionLeft))) false false))
    (:Finished
      (Utilities|FlowControl|Delay 0.2)
      (Variables|LocalLogic|SetDoorsOriginalPosition (Transformation|GetWorldLocation (Variables|Default|GetDoors)))
      (Networking|SwitchHasAuthority
        (:Authority
          _)
        (:Remote
          (Niagara|SpawnSystematLocation (Variables|Default|GetNS_ElevatorExit_04_DoorOpen) (Math|Vector|MakeVector _x _y (- _z (stale _elevatorposition))))))
      (bind (_direction_2 _elevatorposition) (Timeline|LowerDoors Play)
        (:Update
          (Transformation|SetWorldLocation (Variables|Default|GetDoors) (Math|Vector|MakeVector _x _y (- _z _elevatorposition)) false false))
        (:Finished
          (Networking|SwitchHasAuthority
            (:Remote
              (Niagara|SpawnSystematLocation (Variables|Default|GetNS_ElevatorExit_05_Door_SlamGround) (Variables|/Script/Engine.SceneComponent|GetRelativeLocation (Variables|Default|GetDefaultSceneRoot))))))))))

(fn CalculateSpawnAmount
  (Variables|LocalLogic|SetAmountToSpawn (Math|Random|RandomIntegerinRange (Math|Float|Truncate (ScalableFloat|GetValueatLevel (Variables|LocalLogic|GetHAND_MinToSpawn))) (Math|Float|Truncate (ScalableFloat|GetValueatLevel (Variables|LocalLogic|GetHAND_MaxToSpawn)))))
  (bind _budget (CallFunction|GetBudgetbyEvent (Variables|Default|GetHD_GruntTableCode)))
  (if (> (Variables|LocalLogic|GetAmountToSpawn) _budget)
    (Variables|LocalLogic|SetAmountToSpawn _budget))
  (return))

(fn DisablePointProviders
  (bind _patrolprovider_ls (Variables|Patrol|GetPatrolProvider_LS))
  (bind _patrolprovider_lg (Variables|Patrol|GetPatrolProvider_LG))
  (Utilities|IsValid _patrolprovider_ls
    (:Is%20Valid
      (Class|FortAthenaPatrolPathPointProvider|DisablePointProvider _patrolprovider_ls))
    (:Is%20Not%20Valid
      _))
  (Utilities|IsValid _patrolprovider_lg
    (:Is%20Valid
      (Class|FortAthenaPatrolPathPointProvider|DisablePointProvider _patrolprovider_lg))
    (:Is%20Not%20Valid
      _))
  (for _array_element (Variables|Patrol|GetPatrolProviders)
    (Utilities|IsValid _array_element
      (:Is%20Valid
        (Class|FortAthenaPatrolPathPointProvider|DisablePointProvider _array_element))))
  (return))

(fn FireSpawnAnalyticEvent
  ; Set Location Attribute
  (bind _returnvalue (Analytics|ConstructAnalyticsAttribute (LocalVariables|GetSpawnAnalyticLocationAttributeName) (Utilities|String|ToString(Vector) (Transformation|GetActorLocation))))
  ; Set SpawnType Attribute
  (bind _returnvalue_2 (Analytics|ConstructAnalyticsAttribute :AttributeName (LocalVariables|GetSpawnAnalyticSpawnTypeAttributeName) :IntValue (Variables|LocalLogic|GetSpawnType) :floatValue -1.0 :BoolValue false))
  (Analytics|FireAnalyticsEventBlueprintArray (LocalVariables|GetSpawnAnalyticEventName) (Utilities|Array|MakeArray _returnvalue _returnvalue_2)))

(fn GetBudgetByEvent
  (params (RowHandleToQuery :type Data%20Table%20Row%20Handle%20Structure))
  (bind _lwm (Variables|Default|GetLWM))
  (Utilities|IsValid _lwm
    (:Is%20Valid
      (LocalVariables|SetCurrentBudget (Class|FortAthenaLivingWorldManager|QueryEventBudget _lwm RowHandleToQuery self)))
    (:Is%20Not%20Valid
      (LocalVariables|SetCurrentBudget 0)))
  (return (LocalVariables|GetCurrentBudget)))

(fn OnRep_ElevatorSpawned
  (Networking|SwitchHasAuthority
    (:Remote
      (if (Variables|LocalLogic|GetElevatorSpawned)
        (CallFunction|TeleportElevatoratEnd)))))

(fn SetOriginalAssetLocations
  (Variables|LocalLogic|SetBaseOriginalPosition (Transformation|GetWorldLocation (Variables|Default|GetBase))))

(fn TeleportElevatorAtEnd
  (if (Math|Vector|Equal(Vector) (Transformation|GetWorldLocation (Variables|Default|GetBase)) (Variables|LocalLogic|GetBaseOriginalPosition) 0.0001)
    (Transformation|SetWorldLocation (Variables|Default|GetBase) (Variables|Default|GetNewBaseLocation) false true)
    (Transformation|SetWorldLocation (Variables|Default|GetDoors) (Variables|Default|GetNewDoorLocation) false true)))

(fn ConstructionScript
  (Components|Sphere|SetSphereRadius (Variables|Default|GetProximityTrigger) (Variables|Default|GetTriggerRadius) true)
  (Networking|SwitchHasAuthority
    (:Remote
      (Components|DestroyComponent (Variables|Default|GetProximityTrigger)) ; Remove Component, should only be used once, and can save perfs for Overlaps.
    )))
