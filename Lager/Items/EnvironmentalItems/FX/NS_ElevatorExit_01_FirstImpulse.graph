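# Niagara effect for the first impulse of an elevator exit: a ground-hugging burst of debris, dust and smoke.
# GPU emitters throw colliding mesh chunks with four renderers, short-lived dark expanding puffs, and small colliding debris sprites that shift from warm tint to grey.
# A CPU LingerSmoke emitter adds drifting, camera-offset flipbook smoke that fades out.
# Most spawn in flattened hemispheres, with loop-age curves driving radius, rate or force.
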NiagaraSystem /Lager/Items/EnvironmentalItems/FX/NS_ElevatorExit_01_FirstImpulse
dependencies {
 usedDataInterfaces [&NiagaraDataInterfaceCurve, &NiagaraDataInterfaceVectorCurve, &NiagaraDataInterfaceColorCurve]
 usedDynamicInputs []
 usedModules [&SystemState]
}
emitters {
 DirectionalBurstMeshes_01 {
  simTarget=GPUComputeSim
  emitterUpdateScript [
   &EmitterState(Life Cycle Mode=::NewEnumerator1, Loop Behavior=::NewEnumerator1, UseLoopDelay=false)
   &SpawnBurst_Instantaneous(Spawn Count=&UniformRangedInt(Minimum=25, Maximum=25), Use Spawn Probability=false)
   &SpawnBurst_Instantaneous(Spawn Count=&UniformRangedInt(Minimum=25, Maximum=25), Spawn Group=1, Use Spawn Probability=false) {module=SpawnBurst_Instantaneous001}
   &SpawnRate(SpawnRate=25, Use Spawn Probability=false)
  ]
  particleSpawnScript [
   /SetVariables(Particles.RandonInt2:NiagaraInt32=&UniformRangedInt(Minimum=0, Maximum=3)) {isSetParameters=true}
   /Niagara/Modules/Spawn/Initialization/InitializeParticle(
    Lifetime=&MaskFloatBySpawnGroup(Masked Float=&UniformRangedFloat(Minimum=1, Maximum=2), Passthrough Float=2)
    Mass=&UniformRangedFloat(Minimum=1000, Maximum=5000)
    Mesh Scale=&UniformRangedVector(Minimum=<0.25 0.15 0.15>, Maximum=<1 0.5 0.5>)
    Module.Write Color=true
    Module.Write Lifetime=true
    Module.Write Mass=true
    Module.Write Material Random=true
    Module.Write Position=false
    Module.Write Scale=true
    Module.Write SpriteRotation=false
    Module.Write SpriteSize=false
    Position=&SimulationPosition()
   )
   /Niagara/Modules/Update/Mass/CalculateSizeAndRotationalInertiaByMass(Calculate Mesh Scale=true, Initial Model Dimensions:Vector3f=<50 50 50>, Module.Manually Enter Density=false) {enabled=false}
   &AddVelocityInCone(
    Cone Angle=90
    Cone Axis=<1 0 0.8>
    Module.Use Velocity Falloff On Cone Axis=true
    Velocity Falloff Away From Cone Axis=0.333
    Velocity Strength=&UniformRangedFloat(Minimum=175, Maximum=1500)
   ) {enabled=false}
   /Niagara/Modules/Spawn/Location/SphereLocation(
    Hemisphere Z=true
    MaskSpawnGroup=false
    Module.Override Randomness=false
    Non Uniform Scale=<1 1 0.1>
    Sphere Radius=&FloatFromCurve(FloatCurve=DataInterface() {propertyValues {Curve {keys [<at 0 1>,<at 0.25 1.25>]}}}, CurveIndex=$Emitter.NormalizedLoopAge, Scale Curve=300)
    Surface Only=false
   )
   &AddVelocityFromPoint(
    Module.Use Falloff Distance=false
    Origin Offset=&UniformRangedVector(Minimum=<-1 -1 -100>, Maximum=<1 1 -100>)
    Origin Offset Coordinate Space=::Simulation
    Velocity Strength=&UniformRangedFloat(Minimum=200, Maximum=300)
   )
   &AddVelocityInCone(
    Cone Axis=<0 0 1>
    Module.Use Velocity Falloff On Cone Axis=false
    Velocity Strength=&FloatFromCurve(FloatCurve=DataInterface() {propertyValues {Curve {keys [<at 0 1>,<at 0.25 0>]}}}, CurveIndex=$System.NormalizedLoopAge, Scale Curve=150)
   ) {module=AddVelocityInCone001}
   /Niagara/Modules/Spawn/Orientation/V2/InitialMeshOrientation(
    Rotation=&UniformRangedVector(Minimum=<0 0 0>, Maximum=<1 1 1>)
    Use Orientation Vector=true
    Use Rotation Vector=true
   )
   &AddRotationalVelocity(Rotation Rate=&UniformRangedVector(Minimum=<-1 -1 -1>, Maximum=<1 1 1>))
   &ApplyInitialForces()
  ]
  particleUpdateScript [
   &ParticleState()
   &GravityForce(Gravity=<0 0 -1200>)
   &Drag(
    Drag=&FloatFromCurve(FloatCurve=DataInterface() {propertyValues {Curve {keys [
         <at 0 0>
         {arriveTangent=-4.1608, interpMode=RCIM_Cubic, leaveTangent=-4.1608, tangentMode=RCTM_User, time=0.25, value=1}
         {arriveTangent=-0.1083, interpMode=RCIM_Cubic, leaveTangent=-0.1083, tangentMode=RCTM_User, time=1, value=0}
        ]}}})
    Use Linear Drag=true
    Use Rotational Drag=false
   )
   &ScaleColor(
    Scale Alpha=&FloatFromCurve(FloatCurve=DataInterface() {propertyValues {Curve {keys [<at -0.0046 1.0113 RCIM_Cubic>,<at 1.0011 0.002 RCIM_Cubic>]}}})
    ScaleA=true
    ScaleRGB=true
   )
   &Collision(
    Advanced Aging Rate=1
    Age Colliding Particles=false
    Collision Enabled=/Niagara/DynamicInputs/Helpers/MaskBoolBySpawnGroup(
     Masked Bool:NiagaraBool=true
     Passthrough Bool:NiagaraBool=&SetBoolByFloatComparison(A=$Particles.Age, B=0.2)
     Spawn Group Mask:NiagaraInt32=0
    )
    Control Roll On Collision=true
    Enable Rest State=false
    GPU Collision Type=::NewEnumerator2
    Module.EnableMaxCollisionCount=false
    Module.Randomize Collision Normal Vector=true
    Randomize Collision Normal=0.5
    Restitution=0.8
    ScaleRateOfRotation=false
   )
   &SolveForcesAndVelocity(Clamp Velocity=false, Limit Acceleration=false)
   &SolveRotationalForcesAndVelocity()
   &ScaleMeshSize(Scale Factor=&VectorFromFloat(Value=&FloatFromCurve(FloatCurve=DataInterface() {propertyValues {Curve {keys [<at 0 0>,<at 0.1361 1.0035>,<at 0.7956 1.0207>,<at 1 0>]}}})))
  ]
  rendererClasses [&NiagaraMeshRendererProperties]
  renderers [
   {rendererClass=&NiagaraMeshRendererProperties, rendererIndex=0}
   {rendererClass=&NiagaraMeshRendererProperties, rendererIndex=1}
   {rendererClass=&NiagaraMeshRendererProperties, rendererIndex=2}
   {rendererClass=&NiagaraMeshRendererProperties, rendererIndex=3}
  ]
 }
 DirectionalBurstPaint {
  simTarget=GPUComputeSim
  emitterUpdateScript [
   &EmitterState(Life Cycle Mode=::NewEnumerator1, Loop Behavior=::NewEnumerator1, Loop Duration=0.2, UseLoopDelay=false)
   &SpawnBurst_Instantaneous(Use Spawn Probability=false)
   &SpawnRate(SpawnRate=1000, Use Spawn Probability=false)
  ]
  particleSpawnScript [
   &InitializeParticle(
    Color=&UniformRangedLinearColor(Minimum=<rgba 0.0469 0.0388 0.0352 0.7>, Maximum=<rgba 0.0469 0.0388 0.0352 0.8>)
    Lifetime=&UniformRangedFloat(Minimum=1, Maximum=2)
    Mass=&UniformRangedFloat(Minimum=150, Maximum=500)
    Module.Write Color=true
    Module.Write Lifetime=true
    Module.Write Mass=true
    Module.Write Material Random=true
    Module.Write Position=false
    Module.Write Scale=false
    Module.Write SpriteRotation=true
    Module.Write SpriteSize=true
    Position=&SimulationPosition()
    Sprite Rotation=&UniformRangedFloat(Minimum=0, Maximum=360)
    Sprite Size=&UniformRangedVector2D(
     Minimum=&Vector2DFromFloat(Value=100)
     Maximum=&Vector2DFromFloat(Value=200)
    )
   )
   &AddVelocityInCone(
    Cone Angle=90
    Cone Axis=<1 0 0.8>
    Module.Use Velocity Falloff On Cone Axis=true
    Velocity Falloff Away From Cone Axis=0.333
    Velocity Strength=&UniformRangedFloat(Minimum=175, Maximum=1200)
   ) {enabled=false}
   &CylinderLocation(
    Cylinder Height=0
    Cylinder Radius=350
    Lathe Profile=DataInterface() {propertyValues {Curve {keys [<at 0 0.451 RCIM_Cubic>, {arriveTangent=0.049, interpMode=RCIM_Cubic, leaveTangent=0.049, time=0.5, value=1}, <at 1 0.5 RCIM_Cubic>]}}}
    MaskSpawnGroup=false
    Module.Enable Lathe Profile=false
    Module.Override Randomness=false
    Module.Surface Only=false
    Non Uniform Scale:Vector3f=<1 1 0.1>
   )
   &AddVelocityInCone(
    Cone Angle=80
    Cone Axis=&RotateVectorByQuaternion(Vector To Rotate=<0 0 1>, Quaternion=<0 0 0 1>)
    Cone Axis Coordinate Space=::Simulation
    Module.Use Velocity Falloff On Cone Axis=true
    Velocity Falloff Away From Cone Axis=0.333
    Velocity Strength=&UniformRangedFloat(Minimum=50, Maximum=100)
   ) {module=AddVelocityInCone001}
   &AddVelocityFromPoint(
    Module.Use Falloff Distance=false
    Origin Offset=&UniformRangedVector(Minimum=<0 0 0>, Maximum=<0 0 0>)
    Origin Offset Coordinate Space=::Simulation
    Velocity Strength=&UniformRangedFloat(Minimum=50, Maximum=100)
   )
  ]
  particleUpdateScript [
   &ParticleState()
   &GravityForce()
   &Drag(Use Linear Drag=true, Use Rotational Drag=false)
   &ScaleColor(
    Scale Alpha=&FloatFromCurve(FloatCurve=DataInterface() {propertyValues {Curve {keys [
         <at 0.2874 1.0044 RCIM_Cubic>
         {arriveTangent=-1.0341, interpMode=RCIM_Cubic, leaveTangent=-1.0341, tangentMode=RCTM_User, time=0.8481, value=0.8046}
         {arriveTangent=-1.7086, interpMode=RCIM_Cubic, leaveTangent=-1.7086, tangentMode=RCTM_User, time=0.9531, value=0.4738}
         <at 1.0011 0.002 RCIM_Cubic>
        ]}}})
    ScaleA=true
    ScaleRGB=true
   )
   &SolveForcesAndVelocity(Clamp Velocity=false, Limit Acceleration=false)
   &ScaleSpriteSize(Scale Factor=&Vector2DFromFloat(Value=&FloatFromCurve(FloatCurve=DataInterface() {propertyValues {Curve {keys [
          {interpMode=RCIM_Cubic, leaveTangent=11.897, tangentMode=RCTM_Break, time=-0.0072, value=-0.006}
          {interpMode=RCIM_Cubic, tangentMode=RCTM_Break, time=0.9863, value=5.2676}
         ]}}})))
  ]
  rendererClasses [&NiagaraSpriteRendererProperties]
  renderers [{rendererClass=&NiagaraSpriteRendererProperties}]
 }
 HT_Debris {
  simTarget=GPUComputeSim
  emitterUpdateScript [
   &EmitterScalability(
    Module.bApplySpawnCountScale=false
    Module.bApplySpawnCountScaleByDistanceFraction=false
    Module.bUseMaxDistance=true
    Module.bUseMinDistance=false
   ) {enabled=false}
   &EmitterState(Life Cycle Mode=::NewEnumerator1, Loop Behavior=::NewEnumerator1, UseLoopDelay=false)
   &EmitterLifeCycle(MaxLoopCount=1, NextLoopDuration=0.1) {enabled=false}
   &SpawnRate(
    SpawnRate=&Multiply_Float(
     A=1000
     B=&FloatFromCurve(
      FloatCurve=DataInterface() {propertyValues {Curve {keys [<at 0.0033 0.9963>,<at 0.0793 0.923>,<at 0.1437 0.2263>,<at 1 0>]}}}
      CurveIndex=$Emitter.NormalizedLoopAge
     )
    )
    Use Spawn Probability=false
   )
  ]
  particleSpawnScript [
   /SetVariables(Particles.RandomFloat=&UniformRangedFloat(Minimum=0, Maximum=3)) {isSetParameters=true}
   &InitializeParticle(
    Color=&UniformRangedLinearColor(Minimum=<rgba 0 0 0 1>, Maximum=<rgba 0.1927 0.1724 0.1445 0.4>)
    Lifetime=&UniformRangedFloat(Minimum=1, Maximum=2)
    Module.Write Color=true
    Module.Write Lifetime=true
    Module.Write Mass=true
    Module.Write Material Random=true
    Module.Write Position=false
    Module.Write Scale=false
    Module.Write SpriteRotation=true
    Module.Write SpriteSize=true
    Position=&SimulationPosition()
    Sprite Rotation=&UniformRangedFloat(Minimum=0, Maximum=360)
    Sprite Size=&Vector2DFromFloat(Value=&UniformRangedFloat(Minimum=5, Maximum=10))
   )
   &SystemLocation(MaskSpawnGroup=false, Offset=<0 0 2>)
   &SphereLocation(
    Hemisphere Z=true
    MaskSpawnGroup=false
    Module.Override Randomness=false
    Non Uniform Scale=<1 1 0.1>
    Sphere Radius=&FloatFromCurve(FloatCurve=DataInterface() {propertyValues {Curve {keys [<at 0 1>,<at 0.25 1.25>]}}}, CurveIndex=$Emitter.NormalizedLoopAge, Scale Curve=500)
    Surface Only=false
   )
   &AddVelocityInCone(
    Cone Angle=80
    Cone Axis=&RotateVectorByQuaternion(Vector To Rotate=<0 0 1>, Quaternion=<0 0 0 1>)
    Cone Axis Coordinate Space=::Simulation
    Module.Use Velocity Falloff On Cone Axis=true
    Velocity Falloff Away From Cone Axis=0.333
    Velocity Strength=&UniformRangedFloat(Minimum=2, Maximum=15)
   ) {module=AddVelocityInCone001}
   &AccelerationForce(Acceleration=&UniformRangedVector(
     Minimum=<0 0 20>
     Maximum=&MakeVector(
      X=0
      Y=0
      Z=&FloatFromCurve(
       FloatCurve=DataInterface() {propertyValues {Curve {keys [<at 0.0174 1.2>,<at 0.9841 0.2965>]}}}
       CurveIndex=$Emitter.NormalizedLoopAge
       Scale Curve=450
      )
     )
    )) {module=AccelerationForce001}
   &PointForce(
    Force Origin=&SimulationPosition()
    Force Strength=&UniformRangedFloat(
     Minimum=100
     Maximum=&Multiply_Float(
      A=200
      B=&FloatFromCurve(FloatCurve=DataInterface() {propertyValues {Curve {keys [<at 0.0228 1.5928>,<at 0.2468 0.6834>]}}}, CurveIndex=$Emitter.NormalizedLoopAge)
     )
    )
    Module.Use Falloff=false
    Module.UseLinearForce=false
   )
   &ApplyInitialForces()
   &AddVelocityInCone(
    Cone Axis=<0 0 1>
    Module.Use Velocity Falloff On Cone Axis=false
    Velocity Strength=&UniformRangedFloat(Minimum=100, Maximum=400)
   ) {enabled=false}
   /Niagara/Modules/Update/SubUV/SubUVAnimation(Random SubUV Lookup Index:NiagaraFloat=$Particles.RandomFloat, SubUV Animation Mode=::NewEnumerator1) {module=SubUVAnimation001}
  ]
  particleUpdateScript [
   &ParticleState() {module=UpdateAge}
   &Drag(Use Linear Drag=true, Use Rotational Drag=false)
   &AccelerationForce(Acceleration=<0 0 -980>) {enabled=false}
   &GravityForce()
   &Collision(
    Advanced Aging Rate=1
    Age Colliding Particles=false
    Enable Rest State=false
    Friction=0
    GPU Collision Type=::NewEnumerator2
    Module.EnableMaxCollisionCount=false
    Module.Randomize Collision Normal Vector=true
    Particle Radius Scale=0.25
    Randomize Collision Normal=0.5
    Restitution=0.7
   )
   &SolveForcesAndVelocity(Clamp Velocity=false, Limit Acceleration=false)
   &SubUVAnimation(SubUV Animation Mode=::NewEnumerator1) {enabled=false}
   &SpriteRotationRate(Rotation Rate=&Multiply_Float(A=$Particles.RandomFloat, B=4)) {enabled=false}
   &Color() {enabled=false}
   &ScaleSpriteSize(Scale Factor=&Vector2DFromFloat(Value=&FloatFromCurve(FloatCurve=DataInterface() {propertyValues {Curve {keys [<at 0 0>,<at 0.1 1>,<at 0.6514 0.9748>,<at 1 0>]}}}, Scale Curve=2)))
   &ScaleColor(
    Scale Mode=::NewEnumerator1
    Scale RGBA=&Vector4FromCurve(Vector4Curve=DataInterface() {propertyValues {
       AlphaCurve {keys [<at 0 1>,<at 0.8094 1>,<at 1 0>]}
       BlueCurve {keys [<at 0 0.3765>,<at 0.1558 1>,<at 0.9349 0.3385>]}
       GreenCurve {keys [<at 0 0.5608>,<at 0.1558 1>,<at 0.9349 0.3385>]}
       RedCurve {keys [<at 0 1>,<at 0.1558 1>,<at 0.9349 0.3385>]}
      }})
    ScaleRGBA=true
   )
  ]
  rendererClasses [&NiagaraSpriteRendererProperties]
  renderers [{rendererClass=&NiagaraSpriteRendererProperties}]
 }
 LingerSmoke {
  emitterUpdateScript [
   &EmitterState(Life Cycle Mode=::NewEnumerator1, Loop Behavior=::NewEnumerator1, UseLoopDelay=true)
   &SpawnRate(
    SpawnRate=&FloatFromCurve(FloatCurve=DataInterface() {propertyValues {Curve {keys [<at 0 1>,<at 0.25 1.25>]}}}, CurveIndex=$Emitter.NormalizedLoopAge, Scale Curve=10)
    Use Spawn Probability=false
   )
  ]
  particleSpawnScript [
   &InitializeParticle(
    Color=&UniformRangedLinearColor(Minimum=<rgba 0.0573 0.0483 0.043 0.4>, Maximum=<rgba 0.1 0.0895 0.075 0.6>)
    Lifetime=&Multiply_Float(
     A=&UniformRangedFloat(Minimum=2, Maximum=3)
     B=&FloatFromCurve(FloatCurve=DataInterface() {propertyValues {Curve {keys [<at 0 1>,<at 0.25 2>]}}}, CurveIndex=$Emitter.NormalizedLoopAge)
    )
    Module.Write Color=true
    Module.Write Lifetime=true
    Module.Write Mass=true
    Module.Write Material Random=true
    Module.Write Position=false
    Module.Write Scale=false
    Module.Write SpriteRotation=true
    Module.Write SpriteSize=true
    Position=&SimulationPosition()
    Sprite Rotation=&UniformRangedFloat(Minimum=0, Maximum=360)
    Sprite Size=&Vector2DFromFloat(Value=&Multiply_Float(
      A=&UniformRangedFloat(Minimum=300, Maximum=500)
      B=&FloatFromCurve(FloatCurve=DataInterface() {propertyValues {Curve {keys [<at 0 1>,<at 0.2 0.5>]}}}, CurveIndex=$Emitter.NormalizedLoopAge, Scale Curve=2)
     ))
   )
   &AddVelocity(Velocity=&UniformRangedVector(Minimum=<-250 -250 0>, Maximum=<250 250 200>))
   &BoxLocation(Box Size=<350 100 100>, MaskSpawnGroup=false, Module.Override Seed=false, Rotate Box=false, Surface Only=false) {enabled=false}
   &SphereLocation(
    Hemisphere Z=true
    MaskSpawnGroup=false
    Module.Override Randomness=false
    Non Uniform Scale=<1 1 0.1>
    Sphere Radius=&FloatFromCurve(FloatCurve=DataInterface() {propertyValues {Curve {keys [<at 0 1>,<at 0.25 1.25>]}}}, CurveIndex=$Emitter.NormalizedLoopAge, Scale Curve=500)
    Surface Only=false
   )
   &AddVelocityFromPoint(Module.Use Falloff Distance=false, Velocity Strength=10)
  ]
  particleUpdateScript [
   &ParticleState()
   &Drag(Drag=4.5, Use Linear Drag=true, Use Rotational Drag=false)
   &AccelerationForce(Acceleration=&VectorFromCurve(VectorCurve=DataInterface() {propertyValues {
       XCurve {keys [<at -0.0011 1.0067>,<at 0.9989 0>]}
       YCurve {keys [<at -0.0011 1.0067>,<at 0.9989 0>]}
       ZCurve {keys [{arriveTangent=-3701.2285, interpMode=RCIM_Cubic, leaveTangent=-3701.228, tangentMode=RCTM_User, time=0.125, value=0}, <at 1 -1250>]}
      }})) {enabled=false}
   &SolveForcesAndVelocity(Clamp Velocity=false, Limit Acceleration=false)
   &CameraOffset(Camera Offset Amount=&FloatFromCurve(FloatCurve=DataInterface() {propertyValues {Curve {keys [{arriveTangent=584.4823, interpMode=RCIM_Cubic, leaveTangent=584.4823, tangentMode=RCTM_User, time=0, value=80}, <at 1 350>]}}}))
   &ScaleSpriteSize(Scale Factor=&Vector2DFromFloat(Value=&FloatFromCurve(FloatCurve=DataInterface() {propertyValues {Curve {keys [
          {arriveTangent=7.4819, interpMode=RCIM_Cubic, leaveTangent=7.4819, tangentMode=RCTM_User, time=0, value=0.4215}
          <at 0.0979 0.9367>
          <at 1 1>
         ]}}})))
   &ScaleColor(
    Scale Alpha=&Multiply_Float(
     A=&FloatFromCurve(FloatCurve=DataInterface() {propertyValues {Curve {keys [
          {interpMode=RCIM_Cubic, tangentMode=RCTM_User, time=0, value=0}
          {interpMode=RCIM_Cubic, tangentMode=RCTM_User, time=0.15, value=1}
          <at 1 0>
         ]}}})
     B=&FloatFromCurve(
      FloatCurve=DataInterface() {propertyValues {Curve {keys [<at 0 1>,<at 0.4251 0.4256>,<at 1 0>]}}}
      CurveIndex=$Emitter.NormalizedLoopAge
      Scale Curve=&MakeCustomFloatFromBool(
       Bool=&SetBoolByIntComparison(A=0, B=$Emitter.LoopCount, Comparison Type=::NewEnumerator4)
       True Float=2
       False Float=0
      )
     )
    )
    Scale RGB=&VectorFromCurve(VectorCurve=DataInterface() {propertyValues {
       XCurve {keys [{arriveTangent=-2.5342, interpMode=RCIM_Cubic, leaveTangent=-2.5342, tangentMode=RCTM_User, time=0, value=1}, <at 0.25 0.5>]}
       YCurve {keys [{arriveTangent=-2.605, interpMode=RCIM_Cubic, leaveTangent=-2.605, tangentMode=RCTM_User, time=0, value=1}, <at 0.25 0.5>]}
       ZCurve {keys [{arriveTangent=-2.412, interpMode=RCIM_Cubic, leaveTangent=-2.412, tangentMode=RCTM_User, time=0, value=1}, <at 0.25 0.5>]}
      }})
    ScaleA=true
    ScaleRGB=true
   )
   &SubUVAnimation(Number Of Frames=64) {enabled=false}
   /SetVariables(Particles.SubImageIndex=&FloatFromCurve(FloatCurve=DataInterface() {propertyValues {Curve {keys [
         {arriveTangent=68.1678, interpMode=RCIM_Cubic, leaveTangent=55.984, tangentMode=RCTM_User, time=0, value=16}
         {arriveTangent=21.5064, interpMode=RCIM_Cubic, leaveTangent=21.5065, tangentMode=RCTM_User, time=0.9622, value=63}
        ]}}})) {isSetParameters=true}
  ]
  rendererClasses [&NiagaraSpriteRendererProperties]
  renderers [{rendererClass=&NiagaraSpriteRendererProperties}]
 }
}
summary {
 systemName=NS_ElevatorExit_01_FirstImpulse
 emitters [
  {emitterName=LingerSmoke, rendererClasses [&NiagaraSpriteRendererProperties]}
  {emitterName=DirectionalBurstMeshes_01, simTarget=GPUComputeSim, rendererClasses [&NiagaraMeshRendererProperties]}
  {emitterName=DirectionalBurstPaint, simTarget=GPUComputeSim, rendererClasses [&NiagaraSpriteRendererProperties]}
  {emitterName=HT_Debris, simTarget=GPUComputeSim, rendererClasses [&NiagaraSpriteRendererProperties]}
 ]
 userVariables [User.windowSize {type=Vector3f, description="", defaultValue {struct=Vector3f, value=<0 0 0>}}]
}
