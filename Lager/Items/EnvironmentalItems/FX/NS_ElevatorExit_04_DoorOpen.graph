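# Niagara system for lingering smoke, named as the door-open effect of an elevator exit.
# One sprite emitter, LingerSmoke, combines steady spawning with a delayed burst, placing large, dark, translucent puffs using cylinder and box locations.
# A weak point-based push and drag move the puffs while curves slow their spin, grow them, fade alpha and increase camera offset.
# A curve sets SubImageIndex directly to play a flipbook, replacing the disabled SubUVAnimation module.
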NiagaraSystem /Lager/Items/EnvironmentalItems/FX/NS_ElevatorExit_04_DoorOpen
dependencies {
 usedDataInterfaces [&NiagaraDataInterfaceCurve, &NiagaraDataInterfaceVectorCurve]
 usedDynamicInputs []
 usedModules [&SystemState]
}
emitters {LingerSmoke {
  emitterUpdateScript [
   &EmitterState(Life Cycle Mode=::NewEnumerator1, Loop Behavior=::NewEnumerator1, Loop Duration=1.1, UseLoopDelay=true)
   &SpawnBurst_Instantaneous(
    Spawn Count=&UniformRangedInt(Minimum=15, Maximum=15)
    Spawn Group=1
    Spawn Time=1
    Use Spawn Probability=false
   ) {module=SpawnBurst_Instantaneous001}
   &SpawnRate(SpawnRate=50, Use Spawn Probability=false)
  ]
  particleSpawnScript [
   /Niagara/Modules/Spawn/Initialization/InitializeParticle(
    Color=&UniformRangedLinearColor(Minimum=<rgba 0.0573 0.0483 0.043 0.25>, Maximum=<rgba 0.1 0.0895 0.075 0.5>)
    Lifetime=&UniformRangedFloat(Minimum=3, Maximum=5)
    Module.Write Color=true
    Module.Write Lifetime=true
    Module.Write Mass=true
    Module.Write Material Random=true
    Module.Write Position=false
    Module.Write Scale=false
    Module.Write SpriteRotation=true
    Module.Write SpriteSize=true
    Position=&SimulationPosition()
    Sprite Rotation=&UniformRangedFloat(Minimum=0, Maximum=360)
    Sprite Size=&Vector2DFromFloat(Value=&UniformRangedFloat(Minimum=80, Maximum=150))
   )
   /Niagara/Modules/Spawn/Location/SphereLocation(
    Hemisphere Z=true
    MaskSpawnGroup=false
    Module.Override Randomness=false
    Non Uniform Scale=<1 1 0.1>
    Sphere Radius=500
    Surface Only=false
   ) {enabled=false}
   &AddVelocityFromPoint(Module.Use Falloff Distance=false, Velocity Strength=10)
   &CylinderLocation(
    Cylinder Height=10
    Cylinder Radius=300
    MaskSpawnGroup=false
    Module.Enable Lathe Profile=false
    Module.Override Randomness=false
    Module.Surface Only=true
    Use Endcaps In Surface Only Mode=false
   )
   &BoxLocation(Box Size=<100 100 0>, MaskSpawnGroup=false, Module.Override Seed=false, Rotate Box=false, Surface Only=false)
  ]
  particleUpdateScript [
   &ParticleState()
   &Drag(Drag=4.5, Use Linear Drag=true, Use Rotational Drag=false)
   &AccelerationForce(Acceleration=&VectorFromCurve(VectorCurve=DataInterface() {propertyValues {
       XCurve {keys [<at -0.0011 1.0067>,<at 0.9989 0>]}
       YCurve {keys [<at -0.0011 1.0067>,<at 0.9989 0>]}
       ZCurve {keys [{arriveTangent=-3701.2285, interpMode=RCIM_Cubic, leaveTangent=-3701.228, tangentMode=RCTM_User, time=0.125, value=0}, <at 1 -1250>]}
      }})) {enabled=false}
   &SpriteRotationRate(Rotation Rate=&Multiply_Float(
     A=&UniformRangedFloat(Minimum=1, Maximum=1)
     B=&FloatFromCurve(
      FloatCurve=DataInterface() {propertyValues {Curve {keys [{arriveTangent=-3.245, interpMode=RCIM_Cubic, leaveTangent=-3.245, tangentMode=RCTM_User, time=0, value=1}, <at 0.1 0.05>, <at 1.0011 0.01>]}}}
      Scale Curve=350
     )
    ))
   &SolveForcesAndVelocity(Clamp Velocity=false, Limit Acceleration=false)
   &CameraOffset(Camera Offset Amount=&FloatFromCurve(FloatCurve=DataInterface() {propertyValues {Curve {keys [{arriveTangent=584.4823, interpMode=RCIM_Cubic, leaveTangent=584.4823, tangentMode=RCTM_User, time=0, value=80}, <at 1 350>]}}}))
   &ScaleSpriteSize(Scale Factor=&Vector2DFromFloat(Value=&FloatFromCurve(FloatCurve=DataInterface() {propertyValues {Curve {keys [
          {arriveTangent=7.4819, interpMode=RCIM_Cubic, leaveTangent=7.4819, tangentMode=RCTM_User, time=0, value=0.4215}
          <at 0.0979 0.9367>
          <at 1 1>
         ]}}})))
   &ScaleColor(
    Scale Alpha=&FloatFromCurve(FloatCurve=DataInterface() {propertyValues {Curve {keys [
         {interpMode=RCIM_Cubic, tangentMode=RCTM_User, time=0.0055, value=0.0143}
         {arriveTangent=-0.9042, interpMode=RCIM_Cubic, leaveTangent=-0.9042, tangentMode=RCTM_User, time=0.1139, value=0.9017}
         <at 1.0011 0.002>
        ]}}})
    Scale RGB=&VectorFromCurve(VectorCurve=DataInterface() {propertyValues {
       XCurve {keys [{arriveTangent=-2.5342, interpMode=RCIM_Constant, leaveTangent=-2.5342, time=0, value=0.5}]}
       YCurve {keys [{arriveTangent=-2.605, interpMode=RCIM_Constant, leaveTangent=-2.605, time=0, value=0.5}]}
       ZCurve {keys [{arriveTangent=-2.412, interpMode=RCIM_Constant, leaveTangent=-2.412, time=0, value=0.5}]}
      }})
    ScaleA=true
    ScaleRGB=true
   )
   /Niagara/Modules/Update/SubUV/SubUVAnimation(Number Of Frames=64) {enabled=false}
   /SetVariables(Particles.SubImageIndex=&FloatFromCurve(FloatCurve=DataInterface() {propertyValues {Curve {keys [
         {arriveTangent=68.1678, interpMode=RCIM_Cubic, leaveTangent=55.984, tangentMode=RCTM_User, time=0, value=16}
         {arriveTangent=21.5064, interpMode=RCIM_Cubic, leaveTangent=21.5065, tangentMode=RCTM_User, time=0.9622, value=63}
        ]}}})) {isSetParameters=true}
  ]
  rendererClasses [&NiagaraSpriteRendererProperties]
  renderers [{rendererClass=&NiagaraSpriteRendererProperties}]
 }}
summary {
 systemName=NS_ElevatorExit_04_DoorOpen
 emitters [{emitterName=LingerSmoke, rendererClasses [&NiagaraSpriteRendererProperties]}]
 userVariables []
}
