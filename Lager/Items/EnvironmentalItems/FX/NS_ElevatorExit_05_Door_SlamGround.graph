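# Niagara impact effect for an elevator exit door slamming into the ground, built from four independent emitters.
# Big_Glow_Mesh bursts one expanding, flashing mesh with animated material parameters, while Streaks bursts a distance-culled ring of short-lived sprites.
# HT_Debris briefly sprays GPU sprites from a widening torus, with gravity, collision and size scaled by System.ScaleDistance.
# LingerSmoke briefly spawns large, drag-slowed smoke sprites with curve-driven flipbook frames and camera offset.
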NiagaraSystem /Lager/Items/EnvironmentalItems/FX/NS_ElevatorExit_05_Door_SlamGround
dependencies {
 usedDataInterfaces [&NiagaraDataInterfaceCurve, &NiagaraDataInterfaceVectorCurve, &NiagaraDataInterfaceColorCurve]
 usedDynamicInputs [/Game/Effects/Niagara/DynamicInputs/Math/FN_Fit]
 usedModules [
  &self:SystemSpawnScript.SystemScriptSource.SystemScriptGraph.NiagaraNodeAssignment_0.SetVariables_#1
  &SystemState
 ]
}
emitters {
 Big_Glow_Mesh {
  emitterUpdateScript [
   &EmitterScalability(
    Module.bApplySpawnCountScale=false
    Module.bApplySpawnCountScaleByDistanceFraction=false
    Module.bUseMaxDistance=false
    Module.bUseMinDistance=false
   ) {enabled=false}
   &EmitterLifeCycle(NextLoopDuration=1) {enabled=false}
   &EmitterState(Life Cycle Mode=::NewEnumerator1, Loop Behavior=::NewEnumerator1, Loop Duration=1, UseLoopDelay=true)
   &SpawnBurst_Instantaneous(Spawn Count=1, Use Spawn Probability=false)
  ]
  particleSpawnScript [
   /Niagara/Modules/Spawn/Initialization/InitializeParticle(
    Color=<rgba 0.2 0.2 0.2 1>
    Lifetime=0.5
    Mesh Scale=<2.5 2.5 2>
    Module.Write Color=true
    Module.Write Lifetime=true
    Module.Write Mass=true
    Module.Write Material Random=true
    Module.Write Position=true
    Module.Write Scale=true
    Module.Write SpriteRotation=false
    Module.Write SpriteSize=false
    Position=&SimulationPosition()
    Sprite Size=<25 25>
   )
   &SystemLocation(MaskSpawnGroup=false, Offset=<0 0 -100>)
  ]
  particleUpdateScript [
   &ParticleState()
   &ScaleColor(
    Scale RGB=&Multiply_VectorByFloat(
     Vector=<2 2 2>
     Float=&FloatFromCurve(FloatCurve=DataInterface() {propertyValues {Curve {keys [
          <at 0 0 RCIM_Cubic>
          {arriveTangent=0.002, interpMode=RCIM_Cubic, leaveTangent=0.002, time=0.0604, value=0.8535}
          <at 1.0011 0.002 RCIM_Cubic>
         ]}}})
    )
    ScaleA=true
    ScaleRGB=true
   )
   &DynamicMaterialParameters(
    Index 0 Param 1=&FloatFromCurve(FloatCurve=DataInterface() {propertyValues {Curve {keys [<at 0 0>,<at 1 1>]}}})
    Index 0 Param 2=0.5
    Module.Param0WriteEnabled=true
    Module.Param1WriteEnabled=true
    Module.Param2WriteEnabled=true
    Module.Param3WriteEnabled=true
   )
   &ScaleMeshSize(Scale Factor=&VectorFromFloat(Value=&FloatFromCurve(FloatCurve=DataInterface() {propertyValues {Curve {keys [<at -0.0071 0.5459>,<at 0.5512 1.0143>,<at 0.7956 1.0207>,<at 0.9891 1.0324>]}}}))) {module=ScaleMeshSize001}
  ]
  rendererClasses [&NiagaraMeshRendererProperties]
  renderers [{rendererClass=&NiagaraMeshRendererProperties, rendererIndex=0}]
 }
 HT_Debris {
  simTarget=GPUComputeSim
  emitterUpdateScript [
   &EmitterState(Life Cycle Mode=::NewEnumerator1, Loop Behavior=::NewEnumerator1, Loop Duration=0.1, UseLoopDelay=false)
   &SpawnRate(
    SpawnRate=&Multiply_Float(
     A=1500
     B=&FloatFromCurve(FloatCurve=DataInterface() {propertyValues {Curve {keys [<at 0.2236 1.0023>,<at 1 0>]}}}, CurveIndex=$Emitter.NormalizedLoopAge)
    )
    Use Spawn Probability=false
   )
   &SpawnBurst_Instantaneous(Use Spawn Probability=false)
  ]
  particleSpawnScript [
   /SetVariables(Particles.RandomFloat=&UniformRangedFloat(Minimum=0, Maximum=3)) {isSetParameters=true}
   &InitializeParticle(
    Color=&UniformRangedLinearColor(Minimum=<rgba 0 0 0 1>, Maximum=<rgba 0.1927 0.1724 0.1445 0.4>)
    Lifetime=&UniformRangedFloat(Minimum=0.5, Maximum=2)
    Module.Write Color=true
    Module.Write Lifetime=true
    Module.Write Mass=true
    Module.Write Material Random=true
    Module.Write Position=false
    Module.Write Scale=false
    Module.Write SpriteRotation=true
    Module.Write SpriteSize=true
    Position=&SimulationPosition()
    Sprite Rotation=&UniformRangedFloat(Minimum=0, Maximum=360)
    Sprite Size=&UniformRangedVector2D(
     Minimum=&Vector2DFromFloat(Value=&Multiply_Float(A=5, B=$System.ScaleDistance))
     Maximum=&Vector2DFromFloat(Value=&Multiply_Float(A=10, B=$System.ScaleDistance))
    )
   )
   &TorusLocation(
    Align Torus With Axis=true
    Coordinate Space:ENiagaraCoordinateSpace=::World
    Large Radius:NiagaraFloat=&FloatFromCurve(
     FloatCurve=DataInterface() {propertyValues {Curve {keys [<at 0 0.35>,<at 0.2 2.5>,<at 0.3 3.5>]}}}
     CurveIndex=$Emitter.NormalizedLoopAge
     Scale Curve=90
    )
    MaskSpawnGroup=false
    Module.Override Randomness=false
    Surface Position=3
    Torus Mode=::NewEnumerator2
    Torus Origin:NiagaraPosition=&Add_Vector(A=Linked() {linkedVariable {name=Particles.Position, type=NiagaraPosition}}, B=<0 0 15>)
   )
   &AddVelocityFromPoint(Module.Use Falloff Distance=false, Origin Offset=<0 0 -100>, Velocity Strength=500)
   &AddVelocityInCone(
    Cone Angle=180
    Cone Axis=&RotateVectorByQuaternion(Vector To Rotate=<0 0 1>, Quaternion=<0 0 0 1>)
    Cone Axis Coordinate Space=::Simulation
    Module.Use Velocity Falloff On Cone Axis=true
    Velocity Falloff Away From Cone Axis=0.333
    Velocity Strength=&UniformRangedFloat(Minimum=1500, Maximum=1500)
   ) {module=AddVelocityInCone001}
   &AccelerationForce(Acceleration=&UniformRangedVector(
     Minimum=<0 0 -10>
     Maximum=&MakeVector(
      X=0
      Y=0
      Z=&FloatFromCurve(
       FloatCurve=DataInterface() {propertyValues {Curve {keys [<at 0.0174 1.2>,<at 0.9841 0.2965>]}}}
       CurveIndex=$Emitter.NormalizedLoopAge
       Scale Curve=450
      )
     )
    )) {module=AccelerationForce001}
   &PointForce(
    Force Origin=&SimulationPosition()
    Force Strength=&UniformRangedFloat(
     Minimum=100
     Maximum=&Multiply_Float(
      A=200
      B=&FloatFromCurve(FloatCurve=DataInterface() {propertyValues {Curve {keys [<at 0.0228 1.5928>,<at 0.9916 -1.2016>]}}}, CurveIndex=$Emitter.NormalizedLoopAge)
     )
    )
    Module.Use Falloff=false
    Module.UseLinearForce=false
   )
   &ApplyInitialForces()
   &AddVelocityInCone(
    Cone Axis=<0 0 1>
    Module.Use Velocity Falloff On Cone Axis=false
    Velocity Strength=&UniformRangedFloat(Minimum=100, Maximum=400)
   ) {enabled=false}
   /Niagara/Modules/Update/SubUV/SubUVAnimation(Random SubUV Lookup Index:NiagaraFloat=$Particles.RandomFloat, SubUV Animation Mode=::NewEnumerator1) {module=SubUVAnimation001}
  ]
  particleUpdateScript [
   &ParticleState() {module=UpdateAge}
   &Drag(Use Linear Drag=true, Use Rotational Drag=false)
   &AccelerationForce(Acceleration=<0 0 -980>) {enabled=false}
   &GravityForce()
   &Collision(
    Advanced Aging Rate=1
    Age Colliding Particles=false
    GPU Collision Type=::NewEnumerator2
    Module.Control Rest State Via Surface Normal=false
    Module.EnableMaxCollisionCount=false
    Module.Randomize Collision Normal Vector=true
    Particle Radius Scale=0.5
    Randomize Collision Normal=0.25
    Rest State Time Range=1
    Restitution=0.25
   )
   &SolveForcesAndVelocity(Clamp Velocity=false, Limit Acceleration=false)
   &SubUVAnimation(SubUV Animation Mode=::NewEnumerator1) {enabled=false}
   &SpriteRotationRate(Rotation Rate=&Multiply_Float(A=$Particles.RandomFloat, B=4)) {enabled=false}
   &Color() {enabled=false}
   &ScaleSpriteSize(Scale Factor=&Vector2DFromFloat(Value=&FloatFromCurve(FloatCurve=DataInterface() {propertyValues {Curve {keys [<at 0 0>,<at 0.03 1>,<at 0.6514 0.9748>,<at 1 0>]}}}, Scale Curve=2)))
   &ScaleColor(
    Scale Mode=::NewEnumerator1
    Scale RGBA=&Vector4FromCurve(Vector4Curve=DataInterface() {propertyValues {
       AlphaCurve {keys [<at 0 1>,<at 0.8094 1>,<at 1 0>]}
       BlueCurve {keys [<at 0 0.3765>,<at 0.1558 1>,<at 0.9349 0.3385>]}
       GreenCurve {keys [<at 0 0.5608>,<at 0.1558 1>,<at 0.9349 0.3385>]}
       RedCurve {keys [<at 0 1>,<at 0.1558 1>,<at 0.9349 0.3385>]}
      }})
    ScaleRGBA=true
   )
   &SpriteRotationRate(Rotation Rate=&RandomRangeFloat(Minimum=-300, Maximum=300)) {module=SpriteRotationRate001}
  ]
  rendererClasses [&NiagaraSpriteRendererProperties]
  renderers [{rendererClass=&NiagaraSpriteRendererProperties}]
 }
 LingerSmoke {
  emitterUpdateScript [
   &EmitterState(Life Cycle Mode=::NewEnumerator1, Loop Behavior=::NewEnumerator1, Loop Duration=0.1, UseLoopDelay=true)
   &SpawnRate(SpawnRate=200, Use Spawn Probability=false)
  ]
  particleSpawnScript [
   &InitializeParticle(
    Color=&UniformRangedLinearColor(Minimum=<rgba 0.2656 0.224 0.1992 0.5>, Maximum=<rgba 0.1927 0.1724 0.1445 0.7>)
    Lifetime=&UniformRangedFloat(Minimum=2, Maximum=3)
    Module.Write Color=true
    Module.Write Lifetime=true
    Module.Write Mass=true
    Module.Write Material Random=true
    Module.Write Position=false
    Module.Write Scale=false
    Module.Write SpriteRotation=true
    Module.Write SpriteSize=true
    Position=&SimulationPosition()
    Sprite Rotation=&UniformRangedFloat(Minimum=0, Maximum=360)
    Sprite Size=&Vector2DFromFloat(Value=&UniformRangedFloat(Minimum=500, Maximum=600))
   )
   /Niagara/Modules/Spawn/Location/SphereLocation(
    Hemisphere Z=true
    MaskSpawnGroup=false
    Module.Override Randomness=false
    Non Uniform Scale=<1 1 0.1>
    Sphere Radius=500
    Surface Only=false
   ) {enabled=false}
   &SystemLocation(MaskSpawnGroup=false, Offset=<0 0 -100>)
   &CylinderLocation(
    Cylinder Height=10
    Cylinder Radius=300
    MaskSpawnGroup=false
    Module.Enable Lathe Profile=false
    Module.Override Randomness=false
    Module.Surface Only=true
    Use Endcaps In Surface Only Mode=false
   )
   &BoxLocation(Box Size=<300 300 0>, MaskSpawnGroup=false, Module.Override Seed=false, Rotate Box=false, Surface Only=true)
   &AddVelocityInCone(Cone Angle=90, Cone Axis=<0 0 1>, Module.Use Velocity Falloff On Cone Axis=false, Velocity Strength=500)
  ]
  particleUpdateScript [
   &ParticleState()
   &Drag(Drag=4.5, Use Linear Drag=true, Use Rotational Drag=false)
   &AccelerationForce(Acceleration=&VectorFromCurve(VectorCurve=DataInterface() {propertyValues {
       XCurve {keys [<at -0.0011 1.0067>,<at 0.9989 0>]}
       YCurve {keys [<at -0.0011 1.0067>,<at 0.9989 0>]}
       ZCurve {keys [{arriveTangent=-3701.2285, interpMode=RCIM_Cubic, leaveTangent=-3701.228, tangentMode=RCTM_User, time=0.125, value=0}, <at 1 -1250>]}
      }})) {enabled=false}
   &SpriteRotationRate(Rotation Rate=&Multiply_Float(
     A=&UniformRangedFloat(Minimum=1, Maximum=1)
     B=&FloatFromCurve(
      FloatCurve=DataInterface() {propertyValues {Curve {keys [{arriveTangent=-3.245, interpMode=RCIM_Cubic, leaveTangent=-3.245, tangentMode=RCTM_User, time=0, value=1}, <at 0.1 0.05>, <at 1.0011 0.01>]}}}
      Scale Curve=350
     )
    ))
   &PointForce(Force Falloff Distance=500, Force Strength=5000, Module.Use Falloff=true, Module.UseLinearForce=false)
   &SolveForcesAndVelocity(Clamp Velocity=false, Limit Acceleration=false)
   &CameraOffset(Camera Offset Amount=&FloatFromCurve(FloatCurve=DataInterface() {propertyValues {Curve {keys [{arriveTangent=584.4823, interpMode=RCIM_Cubic, leaveTangent=584.4823, tangentMode=RCTM_User, time=0, value=80}, <at 1 350>]}}}))
   &ScaleSpriteSize(Scale Factor=&Vector2DFromFloat(Value=&FloatFromCurve(FloatCurve=DataInterface() {propertyValues {Curve {keys [
          {arriveTangent=5.0786, interpMode=RCIM_Cubic, leaveTangent=5.0785, tangentMode=RCTM_User, time=0, value=0.4215}
          <at 0.2061 0.9033>
          <at 1 1>
         ]}}})))
   &ScaleColor(
    Scale Alpha=&FloatFromCurve(
     FloatCurve=DataInterface() {propertyValues {Curve {keys [
         {interpMode=RCIM_Cubic, tangentMode=RCTM_User, time=0.0055, value=0.0143}
         {arriveTangent=-0.9042, interpMode=RCIM_Cubic, leaveTangent=-0.9042, tangentMode=RCTM_User, time=0.0248, value=0.8955}
         <at 0.2487 0.2505>
         <at 1.0011 0.002>
        ]}}}
     Scale Curve=1.5
    )
    Scale RGB=&VectorFromCurve(VectorCurve=DataInterface() {propertyValues {
       XCurve {keys [{arriveTangent=-2.5342, interpMode=RCIM_Constant, leaveTangent=-2.5342, time=0, value=0.5}]}
       YCurve {keys [{arriveTangent=-2.605, interpMode=RCIM_Constant, leaveTangent=-2.605, time=0, value=0.5}]}
       ZCurve {keys [{arriveTangent=-2.412, interpMode=RCIM_Constant, leaveTangent=-2.412, time=0, value=0.5}]}
      }})
    ScaleA=true
    ScaleRGB=true
   )
   &SubUVAnimation(Number Of Frames=64) {enabled=false}
   /SetVariables(Particles.SubImageIndex=&FloatFromCurve(FloatCurve=DataInterface() {propertyValues {Curve {keys [
         {arriveTangent=68.1678, interpMode=RCIM_Cubic, leaveTangent=55.984, tangentMode=RCTM_User, time=0, value=16}
         {arriveTangent=21.5064, interpMode=RCIM_Cubic, leaveTangent=21.5065, tangentMode=RCTM_User, time=0.9622, value=63}
        ]}}})) {isSetParameters=true}
  ]
  rendererClasses [&NiagaraSpriteRendererProperties]
  renderers [{rendererClass=&NiagaraSpriteRendererProperties}]
 }
 Streaks {
  emitterUpdateScript [
   &EmitterState(Enable Distance Culling=true, MaxDistance=8000, Module.bUseMaxDistance=true, Module.bUseMinDistance=false, Scalability Mode=::NewEnumerator1)
   &SpawnBurst_Instantaneous(Spawn Count=25, Use Spawn Probability=false)
  ]
  particleSpawnScript [
   &InitializeParticle(
    Color=<rgba 0.1198 0.1072 0.0898 1>
    Lifetime=&UniformRangedFloat(Minimum=0.25, Maximum=0.35)
    Module.Write Color=true
    Module.Write Lifetime=true
    Module.Write Mass=true
    Module.Write Material Random=true
    Module.Write Position=false
    Module.Write Scale=false
    Module.Write SpriteRotation=false
    Module.Write SpriteSize=true
    Position=&SimulationPosition()
    Sprite Size=&UniformRangedVector2D(Minimum=<80 400>, Maximum=<150 500>)
   )
   &CylinderLocation(
    Coordinate Space:ENiagaraCoordinateSpace=::World
    Cylinder Height=0
    Cylinder Radius=&UniformRangedFloat(Minimum=400, Maximum=400, Randomness Mode=::NewEnumerator1, Fixed Random Seed=true)
    MaskSpawnGroup=false
    Module.Enable Lathe Profile=false
    Module.Override Randomness=true
    Module.Override Seed=true
    Module.Surface Only=true
    Random Seed=3
    Randomness Mode:ENiagaraRandomnessMode=::NewEnumerator1
    Surface Only Band Thickness=0
    Use Endcaps In Surface Only Mode=false
   )
   &BoxLocation(
    Box Size=<600 600 0>
    MaskSpawnGroup=false
    Module.Override Seed=false
    Rotate Box=false
    Surface Expansion Mode:ENiagaraExpansionMode=::NewEnumerator2
    Surface Only=true
    Surface Only Band Thickness=2
   ) {enabled=false}
   &AddVelocityFromPoint(
    Module.Use Falloff Distance=false
    Origin Offset=&UniformRangedVector(Minimum=<0 0 0>, Maximum=<0 0 0>)
    Velocity Strength=&UniformRangedFloat(Minimum=15, Maximum=5)
   )
   &SubUVAnimation(SubUV Animation Mode=::NewEnumerator1)
   &AddVelocity(Scale Added Velocity=<0 0 0.4>, Velocity=<0 0 15>)
   &CameraOffset(Camera Offset Amount=150)
  ]
  particleUpdateScript [
   &ParticleState()
   &Drag(Use Linear Drag=true, Use Rotational Drag=false)
   &SolveForcesAndVelocity(Clamp Velocity=false, Limit Acceleration=false)
   &ScaleColor(
    Scale Alpha=&FloatFromCurve(
     FloatCurve=DataInterface() {propertyValues {Curve {keys [
         {interpMode=RCIM_Cubic, tangentMode=RCTM_User, time=0.0003, value=0.9198}
         {arriveTangent=-0.9201, interpMode=RCIM_Cubic, leaveTangent=-0.9201, tangentMode=RCTM_User, time=0.5792, value=0.6723}
         <at 1 0>
        ]}}}
     Scale Curve=1.25
    )
    ScaleA=true
    ScaleRGB=true
   )
   &ScaleSpriteSize(Scale Factor=&Vector2DFromFloat(Value=&FloatFromCurve(FloatCurve=DataInterface() {propertyValues {Curve {keys [<at -0.0037 0.0763>,<at 1 1>]}}})))
  ]
  rendererClasses [&NiagaraSpriteRendererProperties]
  renderers [{rendererClass=&NiagaraSpriteRendererProperties}]
 }
}
summary {
 systemName=NS_ElevatorExit_05_Door_SlamGround
 emitters [
  {emitterName=LingerSmoke, rendererClasses [&NiagaraSpriteRendererProperties]}
  {emitterName=Big_Glow_Mesh, rendererClasses [&NiagaraMeshRendererProperties]}
  {emitterName=Streaks, rendererClasses [&NiagaraSpriteRendererProperties]}
  {emitterName=HT_Debris, simTarget=GPUComputeSim, rendererClasses [&NiagaraSpriteRendererProperties]}
 ]
 userVariables [User.windowSize {type=Vector3f, description="", defaultValue {struct=Vector3f, value=<0 0 0>}}]
}
