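; Event-triggered knockback ability for the Slider environmental item, hitting everything its target selection returns.
; It commits, runs target selection, then branches on each actor's type through chained casts.
; Pawns are launched with their last movement velocity plus a scalable upward boost.
; Vehicles get an impact call with randomized horizontal and scalable vertical values; other actors are cast to buildings and given a destroy GameplayEffect.
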(event Ability|EventActivateAbilityFromEvent
  (params (EventData :type Gameplay%20Event%20Data%20Structure%20%28by%20ref%29))
  (Ability|CommitAbility)
  (bind (_asynctaskproxy _targetdata) (Ability|Tasks|DoTargetSelection "None")
    (:Targeted
      (for _array_element (Ability|TargetData|GetAllActorsfromTargetData _targetdata)
        ;[1 1 1] Launch Pawns
        (bind _asfort_pawn (Utilities|Casting|CastToFortPawn _array_element)
          (:then
            (bind (_returnvalue_x _returnvalue_y _returnvalue_z) (Math|Vector|BreakVector (Pawn|Components|CharacterMovement|GetLastUpdateVelocity (Variables|/Script/Engine.Character|GetCharacterMovement _asfort_pawn))))
            (Character|LaunchCharacter _asfort_pawn (Math|Vector|MakeVector _returnvalue_x _returnvalue_y (+ _returnvalue_z (ScalableFloat|GetValueatLevel (Variables|Default|GetPlayerLaunchVelocity)))) true true))
          (:CastFailed
            ;[1 1 1] Launch Vehicle
            (bind _asfort_athena_vehicle (Utilities|Casting|CastToFortAthenaVehicle _array_element)
              (:then
                (Audio|BroadcastAbilityImpactatLocation 0 _asfort_athena_vehicle (Math|Vector|MakeVector (Math|Random|RandomFloatinRange 0.0 3e+02) (Math|Random|RandomFloatinRange 0.0 3e+02) (ScalableFloat|GetValueatLevel (Variables|Default|GetVehicleLaunchVelocity))) 777.0))
              (:CastFailed
                ;[1 1 1] Destroy Building
                (bind _asbuilding_actor (Utilities|Casting|CastToBuildingActor _array_element))
                (Ability|ApplyGameplayEffectToTarget (Ability|TargetData|AbilityTargetDatafromActor _asbuilding_actor) (Variables|Default|GetDestroyBuildingGE))))))))
    (:Cancelled
      _))
  (Ability|EndAbility))

(event Ability|EventOnEndAbility
  (params (bWasCancelled :type Boolean)))
